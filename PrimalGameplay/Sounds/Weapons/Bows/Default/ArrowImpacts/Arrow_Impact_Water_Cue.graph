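# Sound cue for a bow arrow hitting water.
# A Random node picks among seven equally weighted water impact waves, and a Modulator above it adds slight pitch and volume variation.
# The cue uses a boosted volume multiplier, an arrow impact attenuation asset, the Fort_Impacts sound class, and both global and local bow concurrency rules.
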SoundCue /PrimalGameplay/Sounds/Weapons/Bows/Default/ArrowImpacts/Arrow_Impact_Water_Cue
root=Modulator_8
nodes [
 {id=Modulator_8, children [Random_7], properties {pitchMax=1.1, pitchMin=0.9, volumeMax=1.1, volumeMin=0.9}}
 {
  id=Random_7
  children [WavePlayer_56, WavePlayer_57, WavePlayer_58, WavePlayer_59, WavePlayer_61, WavePlayer_62, WavePlayer_63]
  properties {weights [1,1,1,1,1,1,1]}
 }
 {id=WavePlayer_56, properties {soundWaveAssetPtr=&/Arrow_Impact_Water_01}}
 {id=WavePlayer_57, properties {soundWaveAssetPtr=&/Arrow_Impact_Water_02}}
 {id=WavePlayer_58, properties {soundWaveAssetPtr=&/Arrow_Impact_Water_03}}
 {id=WavePlayer_59, properties {soundWaveAssetPtr=&/Arrow_Impact_Water_04}}
 {id=WavePlayer_61, properties {soundWaveAssetPtr=&/Arrow_Impact_Water_06}}
 {id=WavePlayer_62, properties {soundWaveAssetPtr=&/Arrow_Impact_Water_07}}
 {id=WavePlayer_63, properties {soundWaveAssetPtr=&/Arrow_Impact_Water_08}}
]
properties {
 attenuationSettings=/Game/Sounds/Attenuation/Fort_Arrow_Impact_Athena_Att
 maxDistance=6000
 soundClassObject=&Fort_Impacts
 volumeMultiplier=4.032
 concurrencySet [/Game/Sounds/Concurrency_Rules/Bow_Global_Conc, /Game/Sounds/Concurrency_Rules/Bow_Local_Conc]
}
