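# Sound cue for bow arrow impacts, varying by surface hit and by whether the shot is fully charged.
# A FullyCharged switch picks one of two modulators, the charged one louder with a higher pitch range, both feeding a shared SurfaceIndex switch.
# That switch selects randomized dirt, flesh, wood, grass, metal or stone waves with pitch and volume variation, dirt being the default.
# Arrow impact attenuation and global and local bow concurrency rules apply.
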SoundCue /PrimalGameplay/Sounds/Weapons/Bows/Default/ArrowImpacts/Arrow_Impacts_Default_Cue
root=Switch_0
nodes [
 {id=Modulator_0, children [Switch_1], properties {pitchMax=1, pitchMin=0.8, volumeMax=0.7, volumeMin=0.5}}
 {id=Modulator_1, children [Random_1], properties {pitchMax=1.1, pitchMin=0.9, volumeMax=1.3, volumeMin=1.1}}
 {id=Modulator_10, children [Random_0], properties {pitchMax=1.1, pitchMin=0.9, volumeMax=1.1, volumeMin=0.9}}
 {id=Modulator_2, children [Switch_1], properties {pitchMax=1.2, pitchMin=0.9, volumeMax=1, volumeMin=0.8}}
 {id=Modulator_5, children [Random_4], properties {pitchMax=1.1, pitchMin=0.9, volumeMax=1.1, volumeMin=0.9}}
 {id=Modulator_6, children [Random_5], properties {pitchMax=1.1, pitchMin=0.9, volumeMax=1.1, volumeMin=0.9}}
 {id=Modulator_7, children [Random_6], properties {pitchMax=1.1, pitchMin=0.9, volumeMax=1.1, volumeMin=0.9}}
 {id=Modulator_9, children [Random_8], properties {pitchMax=1.1, pitchMin=0.9, volumeMax=1.3, volumeMin=1.1}}
 {
  id=Random_0
  children [WavePlayer_0,WavePlayer_1,WavePlayer_2,WavePlayer_3,WavePlayer_4,WavePlayer_5,WavePlayer_6]
  properties {weights [1,1,1,1,1,1,1]}
 }
 {
  id=Random_1
  children [WavePlayer_9, WavePlayer_10, WavePlayer_11, WavePlayer_12, WavePlayer_13, WavePlayer_23]
  properties {weights [1,1,1,1,1,1]}
 }
 {
  id=Random_4
  children [WavePlayer_31, WavePlayer_32, WavePlayer_33, WavePlayer_34, WavePlayer_35, WavePlayer_36, WavePlayer_37]
  properties {weights [1,1,1,1,1,1,1]}
 }
 {id=Random_5, children [WavePlayer_40, WavePlayer_41, WavePlayer_42, WavePlayer_44, WavePlayer_46], properties {weights [1,1,1,1,1]}}
 {
  id=Random_6
  children [WavePlayer_48, WavePlayer_49, WavePlayer_50, WavePlayer_51, WavePlayer_52, WavePlayer_53, WavePlayer_54]
  properties {weights [1,1,1,1,1,1,1]}
 }
 {id=Random_8, children [WavePlayer_65, WavePlayer_66, WavePlayer_67, WavePlayer_69, WavePlayer_70], properties {weights [1,1,1,1,1]}}
 {id=Switch_0, children [Modulator_0,Modulator_0,Modulator_2], properties {intParameterName=FullyCharged}}
 {
  id=Switch_1
  children [Modulator_10,Modulator_10,Modulator_1,Modulator_9,Modulator_5,Modulator_6,Modulator_7,Modulator_10]
  properties {intParameterName=SurfaceIndex}
 }
 {id=WavePlayer_0, properties {soundWaveAssetPtr=&/Arrow_Impact_Dirt_01}}
 {id=WavePlayer_1, properties {soundWaveAssetPtr=&/Arrow_Impact_Dirt_02}}
 {id=WavePlayer_10, properties {soundWaveAssetPtr=&/Arrow_Impact_Flesh_02}}
 {id=WavePlayer_11, properties {soundWaveAssetPtr=&/Arrow_Impact_Flesh_03}}
 {id=WavePlayer_12, properties {soundWaveAssetPtr=&/Arrow_Impact_Flesh_04}}
 {id=WavePlayer_13, properties {soundWaveAssetPtr=&/Arrow_Impact_Flesh_05}}
 {id=WavePlayer_2, properties {soundWaveAssetPtr=&/Arrow_Impact_Dirt_03}}
 {id=WavePlayer_23, properties {soundWaveAssetPtr=&/Arrow_Impact_Flesh_06}}
 {id=WavePlayer_3, properties {soundWaveAssetPtr=&/Arrow_Impact_Dirt_04}}
 {id=WavePlayer_31, properties {soundWaveAssetPtr=&/Arrow_Impact_Grass_01}}
 {id=WavePlayer_32, properties {soundWaveAssetPtr=&/Arrow_Impact_Grass_02}}
 {id=WavePlayer_33, properties {soundWaveAssetPtr=&/Arrow_Impact_Grass_03}}
 {id=WavePlayer_34, properties {soundWaveAssetPtr=&/Arrow_Impact_Grass_04}}
 {id=WavePlayer_35, properties {soundWaveAssetPtr=&/Arrow_Impact_Grass_05}}
 {id=WavePlayer_36, properties {soundWaveAssetPtr=&/Arrow_Impact_Grass_06}}
 {id=WavePlayer_37, properties {soundWaveAssetPtr=&/Arrow_Impact_Grass_07}}
 {id=WavePlayer_4, properties {soundWaveAssetPtr=&/Arrow_Impact_Dirt_05}}
 {id=WavePlayer_40, properties {soundWaveAssetPtr=&/Arrow_Impact_Metal_01}}
 {id=WavePlayer_41, properties {soundWaveAssetPtr=&/Arrow_Impact_Metal_02}}
 {id=WavePlayer_42, properties {soundWaveAssetPtr=&/Arrow_Impact_Metal_03}}
 {id=WavePlayer_44, properties {soundWaveAssetPtr=&/Arrow_Impact_Metal_05}}
 {id=WavePlayer_46, properties {soundWaveAssetPtr=&/Arrow_Impact_Metal_07}}
 {id=WavePlayer_48, properties {soundWaveAssetPtr=&/Arrow_Impact_Stone_01}}
 {id=WavePlayer_49, properties {soundWaveAssetPtr=&/Arrow_Impact_Stone_02}}
 {id=WavePlayer_5, properties {soundWaveAssetPtr=&/Arrow_Impact_Dirt_06}}
 {id=WavePlayer_50, properties {soundWaveAssetPtr=&/Arrow_Impact_Stone_03}}
 {id=WavePlayer_51, properties {soundWaveAssetPtr=&/Arrow_Impact_Stone_04}}
 {id=WavePlayer_52, properties {soundWaveAssetPtr=&/Arrow_Impact_Stone_05}}
 {id=WavePlayer_53, properties {soundWaveAssetPtr=&/Arrow_Impact_Stone_06}}
 {id=WavePlayer_54, properties {soundWaveAssetPtr=&/Arrow_Impact_Stone_07}}
 {id=WavePlayer_6, properties {soundWaveAssetPtr=&/Arrow_Impact_Dirt_07}}
 {id=WavePlayer_65, properties {soundWaveAssetPtr=&/Arrow_Impact_Wood_01}}
 {id=WavePlayer_66, properties {soundWaveAssetPtr=&/Arrow_Impact_Wood_02}}
 {id=WavePlayer_67, properties {soundWaveAssetPtr=&/Arrow_Impact_Wood_03}}
 {id=WavePlayer_69, properties {soundWaveAssetPtr=&/Arrow_Impact_Wood_05}}
 {id=WavePlayer_70, properties {soundWaveAssetPtr=&/Arrow_Impact_Wood_06}}
 {id=WavePlayer_9, properties {soundWaveAssetPtr=&/Arrow_Impact_Flesh_01}}
]
properties {
 attenuationSettings=/Game/Sounds/Attenuation/Fort_Arrow_Impact_Athena_Att
 maxDistance=6000
 pitchMultiplier=0.95
 soundClassObject=&Fort_Impacts
 volumeMultiplier=4.4352
 concurrencySet [/Game/Sounds/Concurrency_Rules/Bow_Global_Conc, /Game/Sounds/Concurrency_Rules/Bow_Local_Conc]
}
